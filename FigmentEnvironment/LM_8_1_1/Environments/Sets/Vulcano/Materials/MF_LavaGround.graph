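# Material function that builds material attributes for Vulcano lava ground from diffuse, packed mask and normal textures, with mask height as a separate output.
# A static switch picks world-aligned projection, warped by sine noise to break up tiling, or UV sampling defaulting to world XY.
# The UV path applies bump offset, skipped on mobile feature levels.
# An extra UVMapped output samples the same textures with plain mesh UVs.
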MaterialFunction /FigmentEnvironment/LM_8_1_1/Environments/Sets/Vulcano/Materials/MF_LavaGround
graph {nodes [
  n1 = FunctionInput(Preview=ScalarParameter() {defaultValue=921.5037, parameterName=LavaGround_Size}) {bUsePreviewValueAsDefault=true, inputName=WorldAlignedSize, inputType=FunctionInput_Vector2, sortPriority=10}
  n2 = WorldPosition()
  n3 = MaterialFunctionCall(XY=ComponentMask(None=@n2.XYZ) {g=true, r=true}) {materialFunction=&Swizzle}
  n4 = ScalarParameter() {defaultValue=631.9698, parameterName=WorldspaceNoiseSize}
  n5 = Multiply(A=Sine(None=Divide(A=@n3.YX, B=@n4)), B=ScalarParameter() {defaultValue=42.9308, parameterName=WorldspaceNoiseStrength})
  n6 = Constant()
  n7 = MaterialFunctionCall!(
   Diffuse Texture Object=TextureObject() {texture=/FigmentEnvironment/LM_8_1_1/Environments/Sets/Vulcano/Textures/LavaGround_D}
   Spec Texture Object=TextureObject() {samplerType=SAMPLERTYPE_Masks, texture=/FigmentEnvironment/LM_8_1_1/Environments/Sets/Vulcano/Textures/LavaGround_M}
   TextureSize=AppendVector(A=AppendVector(A=@n1, B=@n1), B=@n1)
   --------------------------
   Use seperate settings for Z projection
   Use seperate texture for Z projection
   Texture Size Z projection
   Z Plane Diffuse Texture Object
   Z Plane Spec Texture Object
   "--------------------------      "
   ProjectionTransitionContrast
   Use Texture Alpha Heightmap to Refine the Transistion
   HeightLerpTransitionContrast
   PickTheDiffuseHeightChannel
   TheDiffuseHeightChannel
   "--------------------------       "
   NormalTextureObject=TextureObject() {samplerType=SAMPLERTYPE_Normal, texture=/FigmentEnvironment/LM_8_1_1/Environments/Sets/Vulcano/Textures/LavaGround_N}
   World Space Normals
   "--------------------------  "
   Export Float 4 Diffuse
   Export Float 4 Spec
   "--------------------------     "
   WorldPosition=Add(A=AppendVector(A=@n5, B=@n6), B=@n2.XYZ)
   World Space Normal
  ) {materialFunction=/Engine/Functions/Engine_MaterialFunctions02/Texturing/WorldAlignedTextures_Complex}
  n8 = WorldPosition()
  n9 = FunctionInput(Preview=Divide(A=ComponentMask(None=@n8.XYZ) {g=true, r=true}, B=ScalarParameter() {defaultValue=921.5037, parameterName=LavaGround_Size})) {bUsePreviewValueAsDefault=true, inputName=UV, inputType=FunctionInput_Vector2}
  n10 = BumpOffset(
   Coordinate=@n9
   Height=TextureSample(UVs=@n9) {samplerType=SAMPLERTYPE_Masks, texture=&LavaGround_M}.G
   HeightRatioInput=ScalarParameter() {defaultValue=0.1, parameterName=BumpOffsetLava}
  )
  n11 = FeatureLevelSwitch(Default=@n10, ES2=@n9, ES3_1=@n9, SM4_REMOVED=@n10)
  n12 = TextureSample(UVs=@n11) {texture=&LavaGround_D}
  n13 = ScalarParameter() {defaultValue=0.1777, parameterName=SpecularLavaGround}
  n14 = MaterialFunctionCall(Float3=StaticSwitchParameter(
    True=@n7.XYZ Spec
    False=TextureSample(UVs=@n11) {samplerType=SAMPLERTYPE_Masks, texture=&LavaGround_M}.RGB
   ) {parameterName=LavaGroundWorldAligned}) {materialFunction=&BreakOutFloat3Components}
  n15 = TextureSample(UVs=@n11) {samplerType=SAMPLERTYPE_Normal, texture=&LavaGround_N}
  FunctionOutput(None=MakeMaterialAttributes!(
    BaseColor=StaticSwitchParameter(True=@n7.XYZ Diffuse Texture, False=@n12.RGB) {parameterName=LavaGroundWorldAligned}
    Metallic
    Specular=Multiply(A=@n13, B=@n14.R)
    Roughness=@n14.B
    Anisotropy
    EmissiveColor
    Opacity
    OpacityMask
    Normal=StaticSwitchParameter(True=@n7.Normal XYZ Texture, False=@n15.RGB) {parameterName=LavaGroundWorldAligned}
    Tangent
    WorldPositionOffset
    SubsurfaceColor
    ClearCoat
    ClearCoatRoughness
    AmbientOcclusion
    Refraction
    CustomizedUVs_0
    CustomizedUVs_1
    CustomizedUVs_2
    CustomizedUVs_3
    CustomizedUVs_4
    CustomizedUVs_5
    CustomizedUVs_6
    CustomizedUVs_7
    PixelDepthOffset
    ShadingModel
    Displacement
   ))
  FunctionOutput(None=@n14.G) {outputName=Height, sortPriority=2}
  n16 = TextureSample() {samplerType=SAMPLERTYPE_Masks, texture=&LavaGround_M}
  FunctionOutput(None=MakeMaterialAttributes(
    BaseColor=TextureSample() {texture=&LavaGround_D}.RGB
    Specular=Multiply(A=@n13, B=@n16.R)
    Roughness=@n16.B
    Normal=TextureSample() {samplerType=SAMPLERTYPE_Normal, texture=&LavaGround_N}.RGB
   )) {outputName=UVMapped}
 ]}
