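# Material function that adds a glowing, darkened lava edge to incoming material attributes, for the Vulcano environment set.
# A vertical world-space gradient, measured from the object's Z snapped down to a 384-unit grid, provides the default, overridable transition mask.
# HeightLerp breaks up that mask using inverted MF_LavaGround height.
# The results drive tinted emissive added to the input and a darkening multiplier on base color.
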MaterialFunction /FigmentEnvironment/LM_8_1_1/Environments/Sets/Vulcano/Materials/MF_LavaEdgeGradientSimple
graph {nodes [
  n1 = GetMaterialAttributes(None=FunctionInput() {inputType=FunctionInput_MaterialAttributes}) {attributeGetTypes [#1, #2]}
  n2 = ObjectPositionWS()
  n3 = Subtract(A=WorldPosition().XYZ, B=@n2)
  n4 = ComponentMask(None=@n2) {b=true}
  n5 = Multiply(A=Frac(None=Divide(A=@n4) {constB=384})) {constB=384}
  n6 = ScalarParameter() {defaultValue=-306.7278, group=Lava, parameterName=EdgeGradientOffset}
  n7 = Divide(
   A=Add(A=ComponentMask(None=@n3) {b=true}, B=Add(A=@n5, B=@n6))
   B=ScalarParameter() {defaultValue=54.307, group=Lava, parameterName=EdgeGradientLength}
  )
  n8 = OneMinus(None=FunctionInput(Preview=Clamp(None=@n7)) {bUsePreviewValueAsDefault=true, inputName=TransitionGradient, inputType=FunctionInput_Scalar, sortPriority=4})
  n9 = ScalarParameter() {defaultValue=28.633, group=Lava, parameterName=EmissiveEdgeContrast2}
  n10 = VectorParameter() {group=Lava, parameterName=GradientColor1, defaultValue=<rgba 0 0 0 1>}
  n11 = VectorParameter() {group=Lava, parameterName=GradientColor2, defaultValue=<rgba 1 0.082 0 1>}
  n12 = Multiply(A=TextureCoordinate(), B=ScalarParameter() {defaultValue=3.8718, group=Lava, parameterName=LavaEdgeTiling})
  n13 = OneMinus(None=MaterialFunctionCall!(
    UV=FunctionInput(Preview=@n12) {bUsePreviewValueAsDefault=true, inputName=UVs, inputType=FunctionInput_Vector2, sortPriority=3}
    WorldAlignedSize
   ) {materialFunction=&/MF_LavaGround}.Height)
  n14 = Power(Base=@n13, Exp=ScalarParameter() {defaultValue=3.1021, group=Lava, parameterName=EmissiveEdgeContrast})
  n15 = MaterialFunctionCall(
   Transition Phase=Multiply(A=@n8, B=ScalarParameter() {defaultValue=0.658, group=Lava, parameterName=EmissiveEdgeHeightLerp})
   Height Texture=@n13
   Contrast=ScalarParameter() {defaultValue=1.1432, group=Lava, parameterName=ContrastEdgeEmissive}
  ) {materialFunction=&HeightLerp}
  n16 = Add(
   A=Multiply(A=VectorParameter() {group=Lava, parameterName=GradientColor3, defaultValue=<rgba 3 0.9406 0 1>}.RGB, B=Power(Base=@n8, Exp=@n9))
   B=Multiply(A=LinearInterpolate(A=@n10.RGB, B=@n11.RGB, Alpha=@n14), B=@n15.Alpha)
  )
  n17 = Clamp(None=LinearInterpolate(
    A=ScalarParameter() {defaultValue=1, group=Lava, parameterName=DarkenEdge1}
    B=ScalarParameter() {defaultValue=-2.615, group=Lava, parameterName=DarkenEdge2}
    Alpha=@n15.Lerp Alpha No Contrast
   ))
  FunctionOutput(None=SetMaterialAttributes!(MaterialAttributes=@n1.MaterialAttributes, EmissiveColor=Add(A=@n1.EmissiveColor, B=@n16), BaseColor=Multiply(A=@n1.BaseColor, B=@n17)) {attributeSetTypes [#1, #2]})
 ]}
