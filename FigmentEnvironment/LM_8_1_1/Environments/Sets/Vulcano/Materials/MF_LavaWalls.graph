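# Material function that outputs material attributes for cave-rock walls in the Vulcano lava environment set.
# WorldAlignedTextures_Complex projects cave-wall diffuse, spec and normal textures at a parameterized size, with only its XY outputs used.
# Vertex alpha blends two rock tints over the diffuse, spec blue scales roughness, specular is a plain scalar, and the normal passes through FlattenNormal.
# The scalar WallSize vector, spec-red specular multiply and WallsBrightness are left unconnected.
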MaterialFunction /FigmentEnvironment/LM_8_1_1/Environments/Sets/Vulcano/Materials/MF_LavaWalls
graph {nodes [
  n1 = ScalarParameter() {defaultValue=1011.2155, parameterName=WallSize}
  AppendVector(A=AppendVector(A=@n1, B=@n1), B=@n1)
  n2 = LinearInterpolate(
   A=VectorParameter() {parameterName=ColorRocksFade, defaultValue=<rgba 0.2708 0.2006 0.097 1>}.RGB
   B=VectorParameter() {parameterName=ColorRocks, defaultValue=<rgba 0.1036 0.1146 0.1071 1>}.RGB
   Alpha=VertexColor().A
  )
  n3 = TextureObject() {samplerType=SAMPLERTYPE_Normal, texture=/FigmentEnvironment/LM_8_1_1/Packages/DS_Fortnite_Terrain_NoLOD/Textures/T_Cave_Exterior_Wall_N}
  n4 = MaterialFunctionCall!(
   Diffuse Texture Object=TextureObject() {texture=/FigmentEnvironment/LM_8_1_1/Packages/DS_Fortnite_Terrain_NoLOD/Textures/T_Cave_Exterior_Wall_D}
   Spec Texture Object=TextureObject() {texture=/FigmentEnvironment/LM_8_1_1/Packages/DS_Fortnite_Terrain_NoLOD/Textures/T_Cave_Exterior_Wall_S}
   TextureSize=VectorParameter() {parameterName=WallSize, defaultValue=<rgba 1200 1200 1200 1>}.RGB
   --------------------------
   Use seperate settings for Z projection
   Use seperate texture for Z projection
   Texture Size Z projection
   Z Plane Diffuse Texture Object
   Z Plane Spec Texture Object
   "--------------------------      "
   ProjectionTransitionContrast=ScalarParameter() {defaultValue=1.1018, parameterName=ProjectionContrast}
   Use Texture Alpha Heightmap to Refine the Transistion
   HeightLerpTransitionContrast
   PickTheDiffuseHeightChannel
   TheDiffuseHeightChannel
   "--------------------------       "
   NormalTextureObject=@n3
   World Space Normals
   "--------------------------  "
   Export Float 4 Diffuse
   Export Float 4 Spec
   "--------------------------     "
   WorldPosition
   World Space Normal
  ) {materialFunction=/Engine/Functions/Engine_MaterialFunctions02/Texturing/WorldAlignedTextures_Complex}
  n5 = ScalarParameter() {defaultValue=0.5524, parameterName=WallSpecular}
  n6 = ScalarParameter() {defaultValue=6.7527, parameterName=WallRoughness}
  n7 = MaterialFunctionCall(Float3=@n4.XY Spec) {materialFunction=&BreakOutFloat3Components}
  n8 = ScalarParameter() {parameterName=FlattenWallNormal}
  FunctionOutput(None=MakeMaterialAttributes!(
    BaseColor=Multiply(A=@n2, B=@n4.XY Diffuse Texture)
    Metallic
    Specular=@n5
    Roughness=Multiply(A=@n6, B=@n7.B)
    Anisotropy
    EmissiveColor
    Opacity
    OpacityMask
    Normal=MaterialFunctionCall(Normal=@n4.Normal XY Texture, Flatness=@n8) {materialFunction=&FlattenNormal}.Result
    Tangent
    WorldPositionOffset
    SubsurfaceColor
    ClearCoat
    ClearCoatRoughness
    AmbientOcclusion
    Refraction
    CustomizedUVs_0
    CustomizedUVs_1
    CustomizedUVs_2
    CustomizedUVs_3
    CustomizedUVs_4
    CustomizedUVs_5
    CustomizedUVs_6
    CustomizedUVs_7
    PixelDepthOffset
    ShadingModel
    Displacement
   ))
  Multiply(A=@n5, B=@n7.R)
  ScalarParameter() {defaultValue=0.1905, parameterName=WallsBrightness}
 ]}
